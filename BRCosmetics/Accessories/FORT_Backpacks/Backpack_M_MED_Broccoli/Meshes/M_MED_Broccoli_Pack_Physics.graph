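# Physics asset for the M_MED_Broccoli backpack, giving its dangling controller, pizza and cheese bones simulated collision shapes and joints.
# A single box covers the pack, and small capsules cover the three dyn_ joints.
# Controller and pizza each constrain to the pack, while cheese chains off the pizza for secondary motion.
# The controller locks twist and one swing axis to move in a single plane, and the cheese locks twist only.
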PhysicsAsset /BRCosmetics/Accessories/FORT_Backpacks/Backpack_M_MED_Broccoli/Meshes/M_MED_Broccoli_Pack_Physics
bodies {
 dyn_cheese_jnt [{shapeType=Capsule, length=2, radius=0.7, center=<0.2866 0 0>, rotation=<rot 90 0 0>}]
 dyn_controller_jnt [{shapeType=Capsule, length=2, radius=1, center=<0.3829 0 0>}]
 dyn_pizza_jnt [{shapeType=Capsule, length=3, radius=2, center=<0.6737 -0.2059 0>, rotation=<rot 90 0 0>}]
 pack [{shapeType=Box, extentX=22.5, extentY=4.5, extentZ=24.5, center=<-2.3807 -6.1415 -7.1073>, rotation=<rot -0.3424 8.3095 2.3451>}]
}
constraints [
 {
  bone1Name=dyn_controller_jnt
  bone2Name=pack
  swing1LimitDegrees=5
  swing1Motion=Locked
  swing2LimitDegrees=7
  twistLimitDegrees=45
  twistMotion=Locked
 }
 {bone1Name=dyn_pizza_jnt, bone2Name=pack, swing1LimitDegrees=35, swing2LimitDegrees=10, twistLimitDegrees=30}
 {bone1Name=dyn_cheese_jnt, bone2Name=dyn_pizza_jnt, swing1LimitDegrees=45, swing2LimitDegrees=45, twistLimitDegrees=45, twistMotion=Locked}
]
