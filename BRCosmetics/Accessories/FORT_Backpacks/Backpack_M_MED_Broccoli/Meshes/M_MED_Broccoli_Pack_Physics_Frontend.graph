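# Physics asset for the M_MED_Broccoli backpack cosmetic, Frontend variant, giving its dangling controller, pizza and cheese joints physical bodies.
# A box body covers the pack, with capsules on the three dyn_ joints.
# The controller constrains to the pack with twist and one swing axis locked, leaving a single narrow swing.
# The pizza constrains to the pack with angular limits, and the cheese chains off the pizza with wider swing limits.
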PhysicsAsset /BRCosmetics/Accessories/FORT_Backpacks/Backpack_M_MED_Broccoli/Meshes/M_MED_Broccoli_Pack_Physics_Frontend
bodies {
 dyn_cheese_jnt [{shapeType=Capsule, length=1.4912, radius=0.8636, shapeName=dyn_cheese_jnt_capsule, center=<1 0 0>}]
 dyn_controller_jnt [{shapeType=Capsule, length=7, radius=0.5, center=<0.3829 0 0>}]
 dyn_pizza_jnt [{shapeType=Capsule, length=2.4941, radius=3.1856, shapeName=dyn_pizza_jnt_capsule, center=<3 0 0>}]
 pack [{shapeType=Box, extentX=22.5, extentY=4.5, extentZ=24.5, center=<-2.38 -6.142 -7.108>, rotation=<rot -0.034 8.3094 2.455>}]
}
constraints [
 {
  bone1Name=dyn_controller_jnt
  bone2Name=pack
  swing1LimitDegrees=0
  swing1Motion=Locked
  swing2LimitDegrees=20
  twistLimitDegrees=0
  twistMotion=Locked
 }
 {bone1Name=dyn_pizza_jnt, bone2Name=pack, swing1LimitDegrees=35, swing2LimitDegrees=10, twistLimitDegrees=30}
 {bone1Name=dyn_cheese_jnt, bone2Name=dyn_pizza_jnt, swing1LimitDegrees=45, swing2LimitDegrees=45, twistLimitDegrees=15}
]
