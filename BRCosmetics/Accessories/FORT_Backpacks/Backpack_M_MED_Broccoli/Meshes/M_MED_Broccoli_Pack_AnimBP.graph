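# Animation Blueprint for the M_MED_Broccoli backpack cosmetic, adding physics-driven secondary motion and a skydiving pose adjustment.
# Two chained RigidBody nodes simulate in root-bone space, with a front-end flag selecting between the gameplay and front-end physics assets.
# While skydiving, blended additive roll offsets then tilt the pack and dyn_pizza_jnt bones.
# The Event Graph holds only unconnected stub nodes.
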AnimBlueprint /BRCosmetics/Accessories/FORT_Backpacks/Backpack_M_MED_Broccoli/Meshes/M_MED_Broccoli_Pack_AnimBP
target_skeleton=&/M_MED_Broccoli_Pack_Skeleton
graphs [
 AnimGraph {nodes [
   n1 = RigidBody!(ComponentPose=Local To Component(LocalPose=Input Pose()), bAlphaBoolEnabled:bool=NOT Boolean(A=Get bIsInFrontEnd()), ExternalForce) -> (Pose) {
    alphaInputType=Bool
    bClampLinearTranslationLimitToRefPose=true
    bOverrideWorldGravity=true
    overridePhysicsAsset=&/M_MED_Broccoli_Pack_Physics
    simulationSpace=BaseBoneSpace
    alphaBoolBlend {blendInTime=0.1, blendOption=Linear, blendOutTime=0.1}
    baseBoneRef {boneName=root}
    componentAppliedLinearAccClamp=<3000 3000 3000>
    componentLinearAccScale=<0.3 0.3 0.3>
    componentLinearVelScale=<0.2 0.2 0.2>
    overrideWorldGravity {z=-800}
   }
   n2 = RigidBody(ComponentPose=@n1, bAlphaBoolEnabled=Get bIsInFrontEnd()) {
    alphaInputType=Bool
    bAlphaBoolEnabled=false
    bClampLinearTranslationLimitToRefPose=true
    bOverrideWorldGravity=true
    overridePhysicsAsset=&/M_MED_Broccoli_Pack_Physics_Frontend
    simulationSpace=BaseBoneSpace
    alphaBoolBlend {blendInTime=0.1, blendOption=Linear, blendOutTime=0.1}
    baseBoneRef {boneName=root}
    componentAppliedLinearAccClamp=<3000 3000 3000>
    componentLinearAccScale=<0.3 0.3 0.3>
    componentLinearVelScale=<0.2 0.2 0.2>
    overrideWorldGravity {z=-800}
   }
   n3 = Get bIsSkyDiving()
   n4 = "Transform (Modify) Bone - Bone: pack"!(ComponentPose=@n2, bAlphaBoolEnabled:bool=@n3, Translation="0,0,0", Rotation="0,0,7", Scale) -> (Pose) {
    alphaInputType=Bool
    bAlphaBoolEnabled=false
    rotationMode=BMM_Additive
    alphaBoolBlend {blendInTime=0.1, blendOption=Linear, blendOutTime=0.1}
    boneToModify {boneName=pack}
    rotation {roll=7}
   }
   n5 = "Transform (Modify) Bone - Bone: dyn_pizza_jnt"!(ComponentPose=@n4, bAlphaBoolEnabled:bool=@n3, Translation="0,0,0", Rotation="0,0,23", Scale="1,1,1") -> (Pose) {
    alphaInputType=Bool
    bAlphaBoolEnabled=false
    rotationMode=BMM_Additive
    alphaBoolBlend {blendInTime=0.1, blendOption=Linear, blendOutTime=0.1}
    boneToModify {boneName=dyn_pizza_jnt}
    rotation {roll=23}
   }
   Output Pose(Result=Component To Local(ComponentPose=@n5)) {layerGroup=None}
  ]}
 EventGraph {nodes [
   TryGetPawnOwner()
   Event BlueprintUpdateAnimation()
  ]}
]
variables [NewVar_0 {type=boolean}]
